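# Unlit material for the Esoto Inverted cosmetic wheel, usable on skeletal meshes, with an animated glow in a tintable custom color.
# Panning noise and gradient samples, rim mask channels and a normal-mapped Fresnel combine into the glow intensity, with vertex color channels exposed as named reroutes.
# A Black parameter blends to an inverted look, where that intensity darkens the surface beneath high-contrast, diagonally scrolling noise highlights.
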Material /VehicleCosmetics/Wheels/Esoto_Inverted/Materials/M_Wheel_Esoto_Inverted
graph {
 properties {
  bAutomaticallySetUsageInEditor=false
  bUseMaterialAttributes=true
  bUsedWithMeshDeformer=true
  bUsedWithSkeletalMesh=true
  shadingModel=MSM_Unlit
  usedShadingModels=Unlit
 }
 nodes [
  n1 = VertexColor()
  NamedRerouteDeclaration(None=@n1.R) {name=VC_R, nodeColor=<rgba 1 0 0.2812 1>}
  NamedRerouteDeclaration(None=@n1.G) {name=VC_G, nodeColor=<rgba 0 1 0 1>}
  NamedRerouteDeclaration(None=@n1.B) {name=VC_B, nodeColor=<rgba 0 0 1 1>}
  n2 = Constant() {r=0.75}
  n3 = Constant()
  n4 = Panner(Coordinate=Add(A=TextureCoordinate() {uTiling=0.25}, B=AppendVector(A=@n2, B=@n3))) {speedY=1}
  n5 = TextureSample(UVs=@n4) {samplerSource=SSM_Wrap_WorldGroupSettings, texture=/VehicleCosmetics/Wheels/_SharedTextures/FX/Universal_Noise_34_RGB}
  n6 = Multiply(
   A=ConstantBiasScale(None=TextureSample(UVs=@n4) {samplerSource=SSM_Wrap_WorldGroupSettings, texture=/VehicleCosmetics/Wheels/_SharedTextures/FX/LinearGradient_Balanced}.G) {scale=1}
   B=NamedRerouteUsage()
  )
  n7 = TextureSampleParameter2D() {
   group=00 Base
   parameterName=Masks
   samplerType=SAMPLERTYPE_LinearColor
   texture=/VehicleCosmetics/Wheels/Esoto_Inverted/Textures/T_EsotoInverted_Rim_RGB
  }
  n8 = Multiply(A=Multiply(A=Multiply(A=@n5.B, B=@n6), B=@n7.G)) {constB=1.5}
  n9 = TextureSampleParameter2D() {
   group=00 Base
   parameterName=Normal
   samplerType=SAMPLERTYPE_Normal
   texture=/VehicleCosmetics/Wheels/Esoto_Inverted/Textures/T_EsotoInverted_Rim_N
  }
  n10 = MaterialFunctionCall!(Normal=@n9.RGB, Exponent=ScalarParameter() {defaultValue=7, group=00 Base, parameterName=Fresnel}) {materialFunction=/VehicleCosmetics/Wheels/Cola_Holographic/Materials/MF_Fresnel_Classic}
  n11 = NamedRerouteUsage()
  n12 = Multiply(A=@n7.B, B=NamedRerouteUsage())
  n13 = Clamp(None=@n12) {maxDefault=0.15}
  n14 = Add(A=Multiply(A=@n13, B=@n5.B) {constB=1.5}, B=@n13)
  n15 = Multiply(A=Multiply(A=@n12) {constB=1.5}, B=@n5.B) {constB=1.5}
  n16 = Add(A=@n8, B=Add(A=Subtract(A=@n10.Result, B=@n11), B=Add(A=@n14, B=@n15)))
  n17 = Multiply(A=VectorParameter() {group=00 Base, parameterName=Custom Color, defaultValue=<rgba 1 0.2918 0 1>}.RGB, B=Add(A=@n8, B=@n16))
  n18 = Panner(Coordinate=TextureCoordinate() {uTiling=5}) {speedX=0.5, speedY=-0.5}
  n19 = ConstantBiasScale(None=Multiply(
    A=TextureSample(UVs=@n18) {samplerSource=SSM_Wrap_WorldGroupSettings, texture=&Universal_Noise_34_RGB}.G
    B=NamedRerouteUsage()
   )) {bias=-0.1, scale=20}
  n20 = Clamp(None=ConstantBiasScale(None=Clamp(None=@n19)) {bias=-2.5, scale=1}) {maxDefault=1.5}
  n21 = LinearInterpolate(Alpha=Multiply(A=@n16) {constB=1.5}) {constA=0.011, constB=0}
  n22 = ScalarParameter() {group=00 Base, parameterName=Black}
  n23 = Add(A=@n20, B=ConstantBiasScale(None=@n21) {bias=-0.1, scale=10})
  SetMaterialAttributes!(
   MaterialAttributes
   Base Color=LinearInterpolate(A=@n17, B=Add(A=@n20, B=@n21), Alpha=@n22)
   Emissive Color=LinearInterpolate(A=@n17, B=Clamp(None=@n23) {maxDefault=2}, Alpha=@n22)
  ) {attributeSetTypes [#1, #2]}
 ]
}
